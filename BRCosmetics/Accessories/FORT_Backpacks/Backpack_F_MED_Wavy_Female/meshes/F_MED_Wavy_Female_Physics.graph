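# Physics asset for the Wavy Female back bling cosmetic's skeletal mesh.
# It defines a single body on the backpack_root bone, made of one angled capsule offset from the bone.
# It has no constraints, so the body gives one rough collision volume with no jointed parts.
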PhysicsAsset /BRCosmetics/Accessories/FORT_Backpacks/Backpack_F_MED_Wavy_Female/meshes/F_MED_Wavy_Female_Physics
bodies {backpack_root [{shapeType=Capsule, length=10.9024, radius=7.2754, center=<-1.9877 -3.6778 0.8265>, rotation=<rot 0.6619 -80.7271 89.2621>}]}
constraints []
