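# Animation Blueprint for the Wavy Female backpack cosmetic, adding physics and procedural sway to its looping animation.
# The sequence feeds a RigidBody node, then chained spring-driven Anim Dynamics nodes for spine, neck, head, arms, hands, hair and eye bones; dyn_hair_04 appears three times.
# Additive bone transforms sway spine, neck and head using rotators the EventGraph builds each update from cycling sines modulated by game-time waves, applied one update late.
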AnimBlueprint /BRCosmetics/Accessories/FORT_Backpacks/Backpack_F_MED_Wavy_Female/meshes/F_MED_Wavy_Female_AnimBP
target_skeleton=&/F_MED_Wavy_Female_Skeleton
graphs [
 AnimGraph {nodes [
   n1 = RigidBody(ComponentPose=Local To Component(LocalPose=Sequence Player 'Anim_F_MED_Wavy_Female'() {sequence=/BRCosmetics/Accessories/FORT_Backpacks/Backpack_F_MED_Wavy_Female/Animation/Anim_F_MED_Wavy_Female})) {
    bOverrideWorldGravity=true
    overridePhysicsAsset=&/F_MED_Wavy_Female_Physics
    simulationSpace=BaseBoneSpace
    alphaBoolBlend {blendOption=Linear}
    baseBoneRef {boneName=backpack_root}
    componentLinearAccScale=<0.5 0.5 0.5>
    componentLinearVelScale=<0.5 0.5 0.5>
    overrideWorldGravity {z=-980}
   }
   n2 = Anim Dynamics - Bone: dyn_spine(ComponentPose=@n1, Alpha=1, ExternalForce="0,0,0") {
    angularDampingOverride=0.999
    angularSpringConstant=2000
    bAngularSpring=true
    bGravityOverrideInSimSpace=true
    bOverrideAngularDamping=true
    bOverrideLinearDamping=true
    bUseGravityOverride=true
    linearDampingOverride=0.999
    alphaBoolBlend {blendOption=Linear}
    boundBone {boneName=dyn_spine}
    componentAppliedLinearAccClamp=<1000 1000 1000>
    componentLinearAccScale=<1 1 1>
    componentLinearVelScale=<1 1 1>
    gravityOverride {z=-980}
    physicsBodyDefinitions [{
      sphereCollisionRadius=0
      boundBone {boneName=dyn_spine}
      boxExtents=<5 10 10>
      constraintSetup {angularTargetAxis=AxisY, angularLimitsMax=<1 15 0>, angularLimitsMin=<-1 -5 -30>, angularTarget=<0 0 0>}
      localJointOffset=<2 0 0>
     }]
   }
   n3 = Anim Dynamics - Bone: dyn_neck(ComponentPose=@n2, Alpha=1, ExternalForce="0,0,0") {
    angularDampingOverride=0.999
    angularSpringConstant=2000
    bAngularSpring=true
    bGravityOverrideInSimSpace=true
    bOverrideAngularDamping=true
    bOverrideLinearDamping=true
    bUseGravityOverride=true
    linearDampingOverride=0.999
    alphaBoolBlend {blendOption=Linear}
    boundBone {boneName=dyn_neck}
    componentAppliedLinearAccClamp=<1000 1000 1000>
    componentLinearAccScale=<1 1 1>
    componentLinearVelScale=<1 1 1>
    gravityOverride {z=-980}
    physicsBodyDefinitions [{
      sphereCollisionRadius=0
      boundBone {boneName=dyn_neck}
      boxExtents=<5 20 20>
      constraintSetup {angularTargetAxis=AxisY, angularLimitsMax=<0 10 5>, angularLimitsMin=<0 -10 -20>, angularTarget=<0 0 0>}
      localJointOffset=<2 0 0>
     }]
   }
   n4 = Anim Dynamics - Bone: dyn_head(ComponentPose=@n3, Alpha=1, ExternalForce="0,0,0") {
    angularDampingOverride=0.999
    angularSpringConstant=2000
    bAngularSpring=true
    bGravityOverrideInSimSpace=true
    bOverrideAngularDamping=true
    bOverrideLinearDamping=true
    bUseGravityOverride=true
    gravityScale=0.5
    linearDampingOverride=0.999
    alphaBoolBlend {blendOption=Linear}
    boundBone {boneName=dyn_head}
    componentAppliedLinearAccClamp=<1000 1000 1000>
    componentLinearAccScale=<1 1 1>
    componentLinearVelScale=<1 1 1>
    gravityOverride {z=-980}
    physicsBodyDefinitions [{
      sphereCollisionRadius=0
      boundBone {boneName=dyn_head}
      boxExtents=<5 20 20>
      constraintSetup {angularTargetAxis=AxisY, angularLimitsMax=<0 10 0>, angularLimitsMin=<0 -30 -20>, angularTarget=<0 0 0>}
      localJointOffset=<2 0 0>
     }]
   }
   n5 = "Transform (Modify) Bone - Bone: dyn_spine"(ComponentPose=@n4, Alpha=0.3, Translation="0,0,0", Rotation=Get Rotations_1() -> (Rotations_1:Rotator), Scale="1,1,1") {rotationMode=BMM_Additive, rotationSpace=BCS_BoneSpace, alphaBoolBlend {blendOption=Linear}, boneToModify {boneName=dyn_spine}}
   n6 = "Transform (Modify) Bone - Bone: dyn_neck"(ComponentPose=@n5, Alpha=2, Translation="0,0,0", Rotation=Get Rotations_2() -> (Rotations_2:Rotator), Scale="1,1,1") {rotationMode=BMM_Additive, rotationSpace=BCS_BoneSpace, alphaBoolBlend {blendOption=Linear}, boneToModify {boneName=dyn_neck}}
   n7 = "Transform (Modify) Bone - Bone: dyn_head"(ComponentPose=@n6, Alpha=0.7, Translation="0,0,0", Rotation=Get Rotations_3() -> (Rotations_3:Rotator), Scale="1,1,1") {rotationMode=BMM_Additive, rotationSpace=BCS_BoneSpace, alphaBoolBlend {blendOption=Linear}, boneToModify {boneName=dyn_head}}
   n8 = Anim Dynamics - Bone: dyn_hand_l(ComponentPose=@n7, Alpha=1, ExternalForce="0,0,0") {
    angularDampingOverride=0.999
    angularSpringConstant=1000
    bAngularSpring=true
    bGravityOverrideInSimSpace=true
    bOverrideAngularDamping=true
    bOverrideLinearDamping=true
    bUseGravityOverride=true
    linearDampingOverride=0.999
    alphaBoolBlend {blendOption=Linear}
    boundBone {boneName=dyn_hand_l}
    componentAppliedLinearAccClamp=<1000 1000 1000>
    componentLinearAccScale=<1 1 1>
    componentLinearVelScale=<1 1 1>
    gravityOverride {z=-500}
    physicsBodyDefinitions [{
      sphereCollisionRadius=0
      boundBone {boneName=dyn_hand_l}
      boxExtents=<1 20 20>
      constraintSetup {
       angularTargetAxis=AxisY
       angularLimitsMax=<0 20 10>
       angularLimitsMin=<0 -5 -15>
       angularTarget=<0 0 0>
       linearAxesMax=<0.2 0.2 0.3>
       linearAxesMin=<-0.2 -0.2 -0.3>
      }
      localJointOffset=<5 0 0>
     }]
   }
   n9 = Anim Dynamics - Bone: dyn_arm_l(ComponentPose=@n8, Alpha=1, ExternalForce="0,0,0") {
    angularDampingOverride=0.999
    angularSpringConstant=1000
    bAngularSpring=true
    bGravityOverrideInSimSpace=true
    bOverrideAngularDamping=true
    bOverrideLinearDamping=true
    bUseGravityOverride=true
    gravityScale=0.5
    linearDampingOverride=0.999
    alphaBoolBlend {blendOption=Linear}
    boundBone {boneName=dyn_arm_l}
    componentAppliedLinearAccClamp=<1000 1000 1000>
    componentLinearAccScale=<1 1 1>
    componentLinearVelScale=<1 1 1>
    gravityOverride {z=-500}
    physicsBodyDefinitions [{
      sphereCollisionRadius=0
      boundBone {boneName=dyn_arm_l}
      boxExtents=<1 20 20>
      constraintSetup {
       angularTargetAxis=AxisY
       angularLimitsMax=<0 30 20>
       angularLimitsMin=<0 15 -20>
       angularTarget=<0 0 0>
       linearAxesMax=<0.2 0.2 0.2>
       linearAxesMin=<-0.2 -0.2 -0.2>
      }
      localJointOffset=<4 0 0>
     }]
   }
   n10 = Anim Dynamics - Bone: dyn_hand_r(ComponentPose=@n9, Alpha=1, ExternalForce="0,0,0") {
    angularDampingOverride=0.999
    angularSpringConstant=1000
    bAngularSpring=true
    bGravityOverrideInSimSpace=true
    bOverrideAngularDamping=true
    bOverrideLinearDamping=true
    bUseGravityOverride=true
    linearDampingOverride=0.999
    alphaBoolBlend {blendOption=Linear}
    boundBone {boneName=dyn_hand_r}
    componentAppliedLinearAccClamp=<1000 1000 1000>
    componentLinearAccScale=<1 1 1>
    componentLinearVelScale=<1 1 1>
    gravityOverride {z=-500}
    physicsBodyDefinitions [{
      sphereCollisionRadius=0
      boundBone {boneName=dyn_hand_r}
      boxExtents=<1 20 20>
      constraintSetup {
       angularTargetAxis=AxisY
       angularLimitsMax=<0 5 10>
       angularLimitsMin=<0 -20 -15>
       angularTarget=<0 0 0>
       linearAxesMax=<0.2 0.2 0.3>
       linearAxesMin=<-0.2 -0.2 -0.3>
      }
      localJointOffset=<5 0 0>
     }]
   }
   n11 = Anim Dynamics - Bone: dyn_arm_r(ComponentPose=@n10, Alpha=1, ExternalForce="0,0,0") {
    angularDampingOverride=0.999
    angularSpringConstant=1000
    bAngularSpring=true
    bGravityOverrideInSimSpace=true
    bOverrideAngularDamping=true
    bOverrideLinearDamping=true
    bUseGravityOverride=true
    gravityScale=0.5
    linearDampingOverride=0.999
    alphaBoolBlend {blendOption=Linear}
    boundBone {boneName=dyn_arm_r}
    componentAppliedLinearAccClamp=<1000 1000 1000>
    componentLinearAccScale=<1 1 1>
    componentLinearVelScale=<1 1 1>
    gravityOverride {z=-500}
    physicsBodyDefinitions [{
      sphereCollisionRadius=0
      boundBone {boneName=dyn_arm_r}
      boxExtents=<1 20 20>
      constraintSetup {
       angularTargetAxis=AxisY
       angularLimitsMax=<0 -15 20>
       angularLimitsMin=<0 -30 -20>
       angularTarget=<0 0 0>
       linearAxesMax=<0.2 0.2 0.2>
       linearAxesMin=<-0.2 -0.2 -0.2>
      }
      localJointOffset=<4 0 0>
     }]
   }
   n12 = Anim Dynamics - Bone: dyn_hair_01(ComponentPose=@n11, Alpha=1) {
    angularDampingOverride=0.999
    angularSpringConstant=1000
    bAngularSpring=true
    bGravityOverrideInSimSpace=true
    bOverrideAngularDamping=true
    bOverrideLinearDamping=true
    bUseGravityOverride=true
    linearDampingOverride=0.999
    alphaBoolBlend {blendOption=Linear}
    boundBone {boneName=dyn_hair_01}
    componentAppliedLinearAccClamp=<1000 1000 1000>
    componentLinearAccScale=<1 1 1>
    componentLinearVelScale=<1 1 1>
    gravityOverride {z=-980}
    physicsBodyDefinitions [{
      sphereCollisionRadius=0
      boundBone {boneName=dyn_hair_01}
      boxExtents=<1 20 20>
      constraintSetup {
       angularTargetAxis=AxisY
       angularLimitsMax=<0 10 10>
       angularLimitsMin=<0 -20 -15>
       angularTarget=<0 0 0>
       linearAxesMax=<0.2 0.2 0.3>
       linearAxesMin=<-0.2 -0.2 -0.3>
      }
      localJointOffset=<5 0 0>
     }]
   }
   n13 = Anim Dynamics - Bone: dyn_hair_02(ComponentPose=@n12, Alpha=1) {
    angularDampingOverride=0.999
    angularSpringConstant=1000
    bAngularSpring=true
    bGravityOverrideInSimSpace=true
    bOverrideAngularDamping=true
    bOverrideLinearDamping=true
    bUseGravityOverride=true
    linearDampingOverride=0.999
    alphaBoolBlend {blendOption=Linear}
    boundBone {boneName=dyn_hair_02}
    componentAppliedLinearAccClamp=<1000 1000 1000>
    componentLinearAccScale=<1 1 1>
    componentLinearVelScale=<1 1 1>
    gravityOverride {z=-980}
    physicsBodyDefinitions [{
      sphereCollisionRadius=0
      boundBone {boneName=dyn_hair_02}
      boxExtents=<1 20 20>
      constraintSetup {
       angularTargetAxis=AxisY
       angularLimitsMax=<0 10 10>
       angularLimitsMin=<0 -20 -15>
       angularTarget=<0 0 0>
       linearAxesMax=<0.2 0.2 0.3>
       linearAxesMin=<-0.2 -0.2 -0.3>
      }
      localJointOffset=<5 0 0>
     }]
   }
   n14 = Anim Dynamics - Bone: dyn_hair_03(ComponentPose=@n13, Alpha=1, ExternalForce="0,0,0") {
    angularDampingOverride=0.999
    angularSpringConstant=1000
    bAngularSpring=true
    bGravityOverrideInSimSpace=true
    bOverrideAngularDamping=true
    bOverrideLinearDamping=true
    bUseGravityOverride=true
    linearDampingOverride=0.999
    alphaBoolBlend {blendOption=Linear}
    boundBone {boneName=dyn_hair_03}
    componentAppliedLinearAccClamp=<1000 1000 1000>
    componentLinearAccScale=<1 1 1>
    componentLinearVelScale=<1 1 1>
    gravityOverride {z=-980}
    physicsBodyDefinitions [{
      sphereCollisionRadius=0
      boundBone {boneName=dyn_hair_03}
      boxExtents=<1 20 20>
      constraintSetup {
       angularTargetAxis=AxisY
       angularLimitsMax=<0 10 10>
       angularLimitsMin=<0 -20 -15>
       angularTarget=<0 0 0>
       linearAxesMax=<0.2 0.2 0.3>
       linearAxesMin=<-0.2 -0.2 -0.3>
      }
      localJointOffset=<5 0 0>
     }]
   }
   n15 = Anim Dynamics - Bone: dyn_hair_04(ComponentPose=@n14, Alpha=1, ExternalForce="0,0,0") {
    angularDampingOverride=0.999
    angularSpringConstant=1000
    bAngularSpring=true
    bGravityOverrideInSimSpace=true
    bOverrideAngularDamping=true
    bOverrideLinearDamping=true
    bUseGravityOverride=true
    linearDampingOverride=0.999
    alphaBoolBlend {blendOption=Linear}
    boundBone {boneName=dyn_hair_04}
    componentAppliedLinearAccClamp=<1000 1000 1000>
    componentLinearAccScale=<1 1 1>
    componentLinearVelScale=<1 1 1>
    gravityOverride {z=-980}
    physicsBodyDefinitions [{
      sphereCollisionRadius=0
      boundBone {boneName=dyn_hair_04}
      boxExtents=<1 20 20>
      constraintSetup {
       angularTargetAxis=AxisY
       angularLimitsMax=<0 10 10>
       angularLimitsMin=<0 -20 -15>
       angularTarget=<0 0 0>
       linearAxesMax=<0.2 0.2 0.3>
       linearAxesMin=<-0.2 -0.2 -0.3>
      }
      localJointOffset=<5 0 0>
     }]
   }
   n16 = Anim Dynamics - Bone: dyn_hair_04(ComponentPose=@n15, Alpha=1, ExternalForce="0,0,0") {
    angularDampingOverride=0.999
    angularSpringConstant=1000
    bAngularSpring=true
    bGravityOverrideInSimSpace=true
    bOverrideAngularDamping=true
    bOverrideLinearDamping=true
    bUseGravityOverride=true
    linearDampingOverride=0.999
    alphaBoolBlend {blendOption=Linear}
    boundBone {boneName=dyn_hair_04}
    componentAppliedLinearAccClamp=<1000 1000 1000>
    componentLinearAccScale=<1 1 1>
    componentLinearVelScale=<1 1 1>
    gravityOverride {z=-980}
    physicsBodyDefinitions [{
      sphereCollisionRadius=0
      boundBone {boneName=dyn_hair_04}
      boxExtents=<1 20 20>
      constraintSetup {
       angularTargetAxis=AxisY
       angularLimitsMax=<0 10 10>
       angularLimitsMin=<0 -20 -15>
       angularTarget=<0 0 0>
       linearAxesMax=<0.2 0.2 0.3>
       linearAxesMin=<-0.2 -0.2 -0.3>
      }
      localJointOffset=<5 0 0>
     }]
   }
   n17 = Anim Dynamics - Bone: dyn_hair_04(ComponentPose=@n16, Alpha=1, ExternalForce="0,0,0") {
    angularDampingOverride=0.999
    angularSpringConstant=1000
    bAngularSpring=true
    bGravityOverrideInSimSpace=true
    bOverrideAngularDamping=true
    bOverrideLinearDamping=true
    bUseGravityOverride=true
    linearDampingOverride=0.999
    alphaBoolBlend {blendOption=Linear}
    boundBone {boneName=dyn_hair_04}
    componentAppliedLinearAccClamp=<1000 1000 1000>
    componentLinearAccScale=<1 1 1>
    componentLinearVelScale=<1 1 1>
    gravityOverride {z=-980}
    physicsBodyDefinitions [{
      sphereCollisionRadius=0
      boundBone {boneName=dyn_hair_04}
      boxExtents=<1 20 20>
      constraintSetup {
       angularTargetAxis=AxisY
       angularLimitsMax=<0 10 10>
       angularLimitsMin=<0 -20 -15>
       angularTarget=<0 0 0>
       linearAxesMax=<0.2 0.2 0.3>
       linearAxesMin=<-0.2 -0.2 -0.3>
      }
      localJointOffset=<5 0 0>
     }]
   }
   n18 = Anim Dynamics - Bone: dyn_eye_01(ComponentPose=@n17, Alpha=1, ExternalForce="0,0,0") {
    angularDampingOverride=0.999
    angularSpringConstant=1000
    bAngularSpring=true
    bGravityOverrideInSimSpace=true
    bOverrideAngularDamping=true
    bOverrideLinearDamping=true
    bUseGravityOverride=true
    linearDampingOverride=0.999
    alphaBoolBlend {blendOption=Linear}
    boundBone {boneName=dyn_eye_01}
    componentAppliedLinearAccClamp=<1000 1000 1000>
    componentLinearAccScale=<1 1 1>
    componentLinearVelScale=<1 1 1>
    gravityOverride {z=-980}
    physicsBodyDefinitions [{
      sphereCollisionRadius=0
      boundBone {boneName=dyn_eye_01}
      boxExtents=<1 20 20>
      constraintSetup {
       angularTargetAxis=AxisY
       angularLimitsMax=<0 10 10>
       angularLimitsMin=<0 -20 -15>
       angularTarget=<0 0 0>
       linearAxesMax=<0.2 0.2 0.3>
       linearAxesMin=<-0.2 -0.2 -0.3>
      }
      localJointOffset=<5 0 0>
     }]
   }
   n19 = Anim Dynamics - Bone: dyn_eye_02(ComponentPose=@n18, Alpha=1, ExternalForce="0,0,0") {
    angularDampingOverride=0.999
    angularSpringConstant=1000
    bAngularSpring=true
    bGravityOverrideInSimSpace=true
    bOverrideAngularDamping=true
    bOverrideLinearDamping=true
    bUseGravityOverride=true
    linearDampingOverride=0.999
    alphaBoolBlend {blendOption=Linear}
    boundBone {boneName=dyn_eye_02}
    componentAppliedLinearAccClamp=<1000 1000 1000>
    componentLinearAccScale=<1 1 1>
    componentLinearVelScale=<1 1 1>
    gravityOverride {z=-980}
    physicsBodyDefinitions [{
      sphereCollisionRadius=0
      boundBone {boneName=dyn_eye_02}
      boxExtents=<1 20 20>
      constraintSetup {
       angularTargetAxis=AxisY
       angularLimitsMax=<0 10 10>
       angularLimitsMin=<0 -20 -15>
       angularTarget=<0 0 0>
       linearAxesMax=<0.2 0.2 0.3>
       linearAxesMin=<-0.2 -0.2 -0.3>
      }
      localJointOffset=<5 0 0>
     }]
   }
   n20 = Anim Dynamics - Bone: dyn_eye_03(ComponentPose=@n19, Alpha=1, ExternalForce="0,0,0") {
    angularDampingOverride=0.999
    angularSpringConstant=1000
    bAngularSpring=true
    bGravityOverrideInSimSpace=true
    bOverrideAngularDamping=true
    bOverrideLinearDamping=true
    bUseGravityOverride=true
    linearDampingOverride=0.999
    alphaBoolBlend {blendOption=Linear}
    boundBone {boneName=dyn_eye_03}
    componentAppliedLinearAccClamp=<1000 1000 1000>
    componentLinearAccScale=<1 1 1>
    componentLinearVelScale=<1 1 1>
    gravityOverride {z=-980}
    physicsBodyDefinitions [{
      sphereCollisionRadius=0
      boundBone {boneName=dyn_eye_03}
      boxExtents=<1 20 20>
      constraintSetup {
       angularTargetAxis=AxisY
       angularLimitsMax=<0 10 10>
       angularLimitsMin=<0 -20 -15>
       angularTarget=<0 0 0>
       linearAxesMax=<0.2 0.2 0.3>
       linearAxesMin=<-0.2 -0.2 -0.3>
      }
      localJointOffset=<5 0 0>
     }]
   }
   n21 = Anim Dynamics - Bone: dyn_eye_04(ComponentPose=@n20, Alpha=1, ExternalForce="0,0,0") {
    angularDampingOverride=0.999
    angularSpringConstant=1000
    bAngularSpring=true
    bGravityOverrideInSimSpace=true
    bOverrideAngularDamping=true
    bOverrideLinearDamping=true
    bUseGravityOverride=true
    linearDampingOverride=0.999
    alphaBoolBlend {blendOption=Linear}
    boundBone {boneName=dyn_eye_04}
    componentAppliedLinearAccClamp=<1000 1000 1000>
    componentLinearAccScale=<1 1 1>
    componentLinearVelScale=<1 1 1>
    gravityOverride {z=-980}
    physicsBodyDefinitions [{
      sphereCollisionRadius=0
      boundBone {boneName=dyn_eye_04}
      boxExtents=<1 20 20>
      constraintSetup {
       angularTargetAxis=AxisY
       angularLimitsMax=<0 10 10>
       angularLimitsMin=<0 -20 -15>
       angularTarget=<0 0 0>
       linearAxesMax=<0.2 0.2 0.3>
       linearAxesMin=<-0.2 -0.2 -0.3>
      }
      localJointOffset=<5 0 0>
     }]
   }
   n22 = Anim Dynamics - Bone: dyn_eye_05(ComponentPose=@n21, Alpha=1, ExternalForce="0,0,0") {
    angularDampingOverride=0.999
    angularSpringConstant=1000
    bAngularSpring=true
    bGravityOverrideInSimSpace=true
    bOverrideAngularDamping=true
    bOverrideLinearDamping=true
    bUseGravityOverride=true
    linearDampingOverride=0.999
    alphaBoolBlend {blendOption=Linear}
    boundBone {boneName=dyn_eye_05}
    componentAppliedLinearAccClamp=<1000 1000 1000>
    componentLinearAccScale=<1 1 1>
    componentLinearVelScale=<1 1 1>
    gravityOverride {z=-980}
    physicsBodyDefinitions [{
      sphereCollisionRadius=0
      boundBone {boneName=dyn_eye_05}
      boxExtents=<1 20 20>
      constraintSetup {
       angularTargetAxis=AxisY
       angularLimitsMax=<0 10 10>
       angularLimitsMin=<0 -20 -15>
       angularTarget=<0 0 0>
       linearAxesMax=<0.2 0.2 0.3>
       linearAxesMin=<-0.2 -0.2 -0.3>
      }
      localJointOffset=<5 0 0>
     }]
   }
   n23 = Anim Dynamics - Bone: dyn_eye_06(ComponentPose=@n22, Alpha=1, ExternalForce="0,0,0") {
    angularDampingOverride=0.999
    angularSpringConstant=1000
    bAngularSpring=true
    bGravityOverrideInSimSpace=true
    bOverrideAngularDamping=true
    bOverrideLinearDamping=true
    bUseGravityOverride=true
    linearDampingOverride=0.999
    alphaBoolBlend {blendOption=Linear}
    boundBone {boneName=dyn_eye_06}
    componentAppliedLinearAccClamp=<1000 1000 1000>
    componentLinearAccScale=<1 1 1>
    componentLinearVelScale=<1 1 1>
    gravityOverride {z=-980}
    physicsBodyDefinitions [{
      sphereCollisionRadius=0
      boundBone {boneName=dyn_eye_06}
      boxExtents=<1 20 20>
      constraintSetup {
       angularTargetAxis=AxisY
       angularLimitsMax=<0 10 10>
       angularLimitsMin=<0 -20 -15>
       angularTarget=<0 0 0>
       linearAxesMax=<0.2 0.2 0.3>
       linearAxesMin=<-0.2 -0.2 -0.3>
      }
      localJointOffset=<5 0 0>
     }]
   }
   n24 = Anim Dynamics - Bone: dyn_eye_l(ComponentPose=@n23, Alpha=1, ExternalForce="0,0,0") {
    angularDampingOverride=0.999
    angularSpringConstant=1000
    bAngularSpring=true
    bGravityOverrideInSimSpace=true
    bOverrideAngularDamping=true
    bOverrideLinearDamping=true
    bUseGravityOverride=true
    linearDampingOverride=0.999
    alphaBoolBlend {blendOption=Linear}
    boundBone {boneName=dyn_eye_l}
    componentAppliedLinearAccClamp=<1000 1000 1000>
    componentLinearAccScale=<1 1 1>
    componentLinearVelScale=<1 1 1>
    gravityOverride {z=-980}
    physicsBodyDefinitions [{
      sphereCollisionRadius=0
      boundBone {boneName=dyn_eye_l}
      boxExtents=<1 20 20>
      constraintSetup {angularTargetAxis=AxisY, angularLimitsMax=<0 5 10>, angularLimitsMin=<0 -5 -8>, angularTarget=<0 0 0>}
      localJointOffset=<5 0 0>
     }]
   }
   n25 = Anim Dynamics - Bone: dyn_eye_r(ComponentPose=@n24, Alpha=1, ExternalForce="0,0,0") {
    angularDampingOverride=0.999
    angularSpringConstant=1000
    bAngularSpring=true
    bGravityOverrideInSimSpace=true
    bOverrideAngularDamping=true
    bOverrideLinearDamping=true
    bUseGravityOverride=true
    linearDampingOverride=0.999
    alphaBoolBlend {blendOption=Linear}
    boundBone {boneName=dyn_eye_r}
    componentAppliedLinearAccClamp=<1000 1000 1000>
    componentLinearAccScale=<1 1 1>
    componentLinearVelScale=<1 1 1>
    gravityOverride {z=-980}
    physicsBodyDefinitions [{
      sphereCollisionRadius=0
      boundBone {boneName=dyn_eye_r}
      boxExtents=<1 20 20>
      constraintSetup {angularTargetAxis=AxisY, angularLimitsMax=<0 5 10>, angularLimitsMin=<0 -5 -8>, angularTarget=<0 0 0>}
      localJointOffset=<5 0 0>
     }]
   }
   Output Pose(Result=Component To Local(ComponentPose=@n25)) {layerGroup=None}
   AnimGraphNode_SubInput_0 = Input Pose()
  ]}
 EventGraph {nodes [
   Event BlueprintUpdateAnimation()
   > n1 = Sequence!(execute) -> (then_0, then_1, then_2:Exec, then_3:Exec, then_4:Exec)
   n2 = GetGameTimeInSeconds()
   n3 = "Sin (Degrees)"!(A:double=float * float(A=@n2, B=150)) -> (ReturnValue:double)
   n4 = "Sin (Degrees)"(A=float * float(A=@n2, B=200))
   n5 = float / float(
    A=float * float(
     A=MapRangeClamped(Value=@n3, InRangeA=-1, InRangeB=1, OutRangeA=-1)
     B=MapRangeClamped(Value=@n4, InRangeA=-1, InRangeB=1, OutRangeA=-40, OutRangeB=40)
    )
    B=1
   ) {is=title K2Node_CallFunction}
   n6 = "Sin (Degrees)"(A=Get Bobbin_Cycle_1() -> (Bobbin_Cycle_1:double))
   Set Rotation_Z_1(
    execute=@n1.then_2
    Rotation_Z_1:double=float * float(A=MapRangeClamped(Value=@n6, InRangeA=-1, InRangeB=1, OutRangeA=-1.5, OutRangeB=0.6), B=@n5)
   ) -> (Output_Get:double)
   n7 = "Sin (Degrees)"(A=Get Bobbin_Cycle_2() -> (Bobbin_Cycle_2:double))
   > Set Rotation_Z_2(Rotation_Z_2:double=float * float(A=MapRangeClamped(Value=@n7, InRangeA=-1, InRangeB=1, OutRangeA=-0.6, OutRangeB=0.6), B=@n5)) -> (Output_Get:double)
   n8 = "Sin (Degrees)"(A=Get Bobbin_Cycle_3() -> (Bobbin_Cycle_3:double))
   > Set Rotation_Z_3(Rotation_Z_3:double=float * float(A=MapRangeClamped(Value=@n8, InRangeA=-1, InRangeB=1, OutRangeA=-0.7, OutRangeB=0.8), B=@n5)) -> (Output_Get:double)
   n9 = float * float(A=GetWorldDeltaSeconds!() -> (ReturnValue:double), B=360)
   n10 = float / float(A=@n9, B=2)
   n11 = Get Bobbin_Cycle_1()
   Set Bobbin_Cycle_1(
    execute=@n1.then_0
    Bobbin_Cycle_1:double="% (Float)"!(A:double=float + float(A=@n11, B=@n10) {is=title K2Node_CommutativeAssociativeBinaryOperator}, B:double=360) -> (ReturnValue:double)
   ) -> (Output_Get:double)
   n12 = float / float(A=@n9, B=2.1)
   n13 = Get Bobbin_Cycle_2()
   > Set Bobbin_Cycle_2(Bobbin_Cycle_2:double="% (Float)"(A=float + float(A=@n13, B=@n12), B=360)) -> (Output_Get:double)
   n14 = float / float(A=@n9, B=2.1)
   n15 = Get Bobbin_Cycle_3()
   > Set Bobbin_Cycle_3(Bobbin_Cycle_3:double="% (Float)"(A=float + float(A=@n15, B=@n14), B=360)) -> (Output_Get:double)
   n16 = "Sin (Degrees)"(A=Get Bobbin_Cycle_1())
   n17 = GetGameTimeInSeconds()
   n18 = "Sin (Degrees)"(A=float * float(A=@n17, B=80))
   n19 = "Sin (Degrees)"(A=float * float(A=@n17, B=80))
   n20 = float / float(
    A=float * float(
     A=MapRangeClamped(Value=@n18, InRangeA=-1, InRangeB=1, OutRangeA=-1)
     B=MapRangeClamped(Value=@n19, InRangeA=-1, InRangeB=1, OutRangeA=-40, OutRangeB=35)
    )
    B=1
   )
   Set Rotation_Y_1(
    execute=@n1.then_3
    Rotation_Y_1:double=float * float(A=MapRangeClamped(Value=@n16, InRangeA=-1, InRangeB=1, OutRangeA=-2, OutRangeB=2), B=@n20)
   ) -> (Output_Get:double)
   n21 = "Sin (Degrees)"(A=Get Bobbin_Cycle_2())
   > Set Rotation_Y_2(Rotation_Y_2:double=float * float(A=MapRangeClamped(Value=@n21, InRangeA=-1, InRangeB=1, OutRangeA=-0.5, OutRangeB=0.5), B=@n20)) -> (Output_Get:double)
   n22 = "Sin (Degrees)"(A=Get Bobbin_Cycle_3())
   > Set Rotation_Y_3(Rotation_Y_3:double=float * float(A=MapRangeClamped(Value=@n22, InRangeA=-1, InRangeB=1, OutRangeA=-0.5, OutRangeB=0.5), B=@n20)) -> (Output_Get:double)
   Set Rotations_1(
    execute=@n1.then_1
    Rotations_1:Rotator=MakeRotator(Roll=-30, Pitch=Get Rotation_Y_1() -> (Rotation_Y_1:double), Yaw=Get Rotation_Z_1() -> (Rotation_Z_1:double))
   ) -> (Output_Get:Rotator)
   > Set Rotations_2(Rotations_2:Rotator=MakeRotator(Roll=-12, Pitch=Get Rotation_Y_2() -> (Rotation_Y_2:double), Yaw=Get Rotation_Z_2() -> (Rotation_Z_2:double))) -> (Output_Get:Rotator)
   > Set Rotations_3(Rotations_3:Rotator=MakeRotator(Pitch=Get Rotation_Y_3() -> (Rotation_Y_3:double), Yaw=Get Rotation_Z_3() -> (Rotation_Z_3:double))) -> (Output_Get:Rotator)
  ]}
]
variables [
 Bobbin_Cycle_1 {type=number}
 Bobbin_Cycle_2 {type=number}
 Bobbin_Cycle_3 {type=number}
 Rotation_Y_1 {type=number}
 Rotation_Y_2 {type=number}
 Rotation_Y_3 {type=number}
 Rotation_Z_1 {type=number}
 Rotation_Z_2 {type=number}
 Rotation_Z_3 {type=number}
 Rotations_1 {type=Rotator}
 Rotations_2 {type=Rotator}
 Rotations_3 {type=Rotator}
]
